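# Unlit masked Niagara sprite material for a cross-shaped sparkle flare in the SpireTend pickaxe effects, with responsive AA enabled.
# A center sphere-mask glow, scaled by a dynamic parameter and sharpened by a power, adds to particle color, then is tinted from EdgeColor toward white by the squared flare mask.
# The color gets HDR remapping except on mobile, then emissive adjustments.
# Particle alpha times the flare mask provides the cutout value.
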Material /BRCosmetics/Weapons/FORT_Melee/Pickaxe_SpireTend/FX/M_Sparkle_Flare_Masked
graph {
 properties {
  bEnableResponsiveAA=true
  bUsedWithNiagaraSprites=true
  blendMode=BLEND_Masked
  opacityMaskClipValue=0.15
  pixelDepthOffsetMode=PDOM_Legacy
  shadingModel=MSM_Unlit
  usedShadingModels=Unlit
 }
 nodes [
  n1 = ParticleColor()
  n2 = Multiply(
   A=SphereMask(A=TextureCoordinate(), B=Constant() {r=0.5}, Radius=ScalarParameter() {defaultValue=0.08, parameterName=CenterGlowRadius}) {hardnessPercent=0}
   B=DynamicParameter() {paramNames [HDRMult, CenterGlowBrightness, Param3, Param4]}.CenterGlowBrightness
  )
  n3 = TextureSampleParameter2D() {parameterName=Mask, samplerType=SAMPLERTYPE_Grayscale, texture=/Game/Effects/Fort_Effects/Textures/Flares/T_Cross_Flare}
  n4 = Multiply(
   A=Add(A=@n1.RGB, B=Power(Base=@n2) {constExponent=4})
   B=LinearInterpolate(A=VectorParameter() {parameterName=EdgeColor, defaultValue=<rgba 0.0733 0.1032 0.276 0>}.RGB, Alpha=Multiply(A=@n3.G, B=@n3.G))
  )
  n5 = Clamp(None=@n4)
  n6 = DynamicParameter() {paramNames [HDRMult, CenterGlowBrightness, Param3, Param4]}
  MaterialFunctionCall(Emissive=ShadingPathSwitch(
    Default=MaterialFunctionCall(LDR Input=@n5, HDR Tint=@n4, HDR Intensity Multiplier=@n6.HDRMult) {materialFunction=&DeriveHDRfromLDR}.RemappedRange
    Mobile=@n4
   )) {materialFunction=&MF_FX_EmissiveAdjustments}
  Multiply(A=@n1.A, B=@n3.G)
 ]
}
