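# Niagara system for the swing trail of the SpireTend pickaxe cosmetic.
# The Trail emitter builds an orange animation-trail ribbon between two indexed bones or sockets, with width set per swing section and user multipliers for width and lifetime.
# Its alpha and front glow scale by time of day, and the glow color differs in the front end.
# The Minimal emitter bursts a persistent mesh particle that samples a socket and writes Lightblade position and timeline-driven color.
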NiagaraSystem /BRCosmetics/Weapons/FORT_Melee/Pickaxe_SpireTend/FX/NS_Pickaxe_SpireTend_Trail
dependencies {
 usedDataInterfaces [
  &NiagaraDataInterfaceArrayInt32
  &NiagaraDataInterfaceCurve
  &NiagaraDataInterfaceArrayFloat
  &NiagaraDataInterfaceColorCurve
  &NiagaraDataInterfaceSocketReader
 ]
 usedDynamicInputs []
 usedModules [
  &self:SystemSpawnScript.SystemScriptSource.SystemScriptGraph.NiagaraNodeAssignment_0.SetVariables_#1
  &SystemState
 ]
}
emitters {
 Minimal {
  emitterUpdateScript [
   &EmitterState(
    Life Cycle Mode=::NewEnumerator1
    Loop Behavior=::NewEnumerator1
    Loop Duration Mode=::NewEnumerator1
    Scalability Mode=::NewEnumerator1
    UseLoopDelay=false
   )
   &SpawnBurst_Instantaneous(Use Spawn Probability=false, UseLoopCountLimit=false)
  ]
  particleSpawnScript [
   &InitializeParticle(Color Mode=::NewEnumerator0, UsePositionOffset=false, Write Material Random=true)
   /SetVariables(Particles.Scale=<0.1 0.1 0.1>) {isSetParameters=true}
   /SetVariables(Particles.Lightblade Camera Offset=1) {isSetParameters=true}
   /SetVariables() {isSetParameters=true}
  ]
  particleUpdateScript [
   &ParticleState(Kill Particles When Lifetime Has Elapsed=false)
   &SocketLocation(
    Position Sampling=::NewEnumerator1
    SamplePosOffset=true
    Socket Index=2
    Socket Reader=$System.Socket Reader
    Socket Sampling Mode=::NewEnumerator3
    VelocitySpeedLimit=false
   )
   &Timeline(Timeline End=0.25)
   /SetVariables(Particles.Lightblade Position:NiagaraPosition=$Output.SocketLocation.SampledPosition) {isSetParameters=true}
   /SetVariables(Particles.Lightblade Color:LinearColor=&ColorFromCurve(
     ColorCurve=DataInterface() {propertyValues {
       AlphaCurve {keys [<at 0 1.05>,<at 0.8568 1.05>,<at 1 0 RCIM_Cubic>]}
       BlueCurve {keys [<at 0 0 RCIM_Cubic>,<at 0.2628 0.0227>,<at 1 0.0423 RCIM_Cubic>]}
       GreenCurve {keys [<at 0 0.4584 RCIM_Cubic>,<at 0.2628 0.2347>,<at 1 0.0786 RCIM_Cubic>]}
       RedCurve {keys [<at 0 1 RCIM_Cubic>,<at 0.2628 0.8727>,<at 1 0.2552 RCIM_Cubic>]}
      }}
     CurveIndex=$Particles.Timeline.NormalizedTime
    )) {isSetParameters=true}
  ]
  rendererClasses [&NiagaraMeshRendererProperties]
  renderers [{rendererClass=&NiagaraMeshRendererProperties, rendererIndex=0}]
 }
 Trail {
  emitterSpawnScript [&FN_RibbonIDOffsetUtility()]
  emitterUpdateScript [
   &EmitterState()
   &FN_SpawnAnimTrail()
  ]
  particleSpawnScript [
   &FN_InitializeAnimTrailBoneIndecies(
    First Bone or Socket Index:NiagaraInt32=&SelectIntFromArray(Direct Array Index=0, Int Selection Array=DataInterface() {propertyValues {IntData [0]}})
    Second Bone or Socket Index:NiagaraInt32=&SelectIntFromArray(Direct Array Index=0, Int Selection Array=DataInterface() {propertyValues {IntData [1]}})
   ) {module=FN_InitializeAnimTrailBoneIndecies002}
   &FN_InitializeAnimTrail(
    AnimTrailWidth=&Multiply_Float(
     A=&SelectFloatFromArray(Direct Array Index=$User.SwingSectionIndex, Float Selection Array=DataInterface() {propertyValues {FloatData [0.5,0.6,0.55,0.55]}})
     B=$User.WidthMult
    )
    Color=&MultiplyLinearColor(
     Linear Color=<rgba 1 0.5117 0 1>
     Scale Factor=&MakeVector4(
      X=1
      Y=1
      Z=1
      W=&FloatFromCurve(
       FloatCurve=DataInterface() {propertyValues {Curve {keys [<at 0 1>,<at 0.1 1>,<at 0.256 1>,<at 0.5 0>]}}}
       Scale Curve=&FNTimeOfDayScalar(Night Scale=4, Morning Scale=3, Day Scale=3, Evening Scale=3)
      )
     )
    )
    Lifetime=&Multiply_Float(A=2, B=$User.TrailLifeTimeMult)
   )
  ]
  particleUpdateScript [
   &SolveForcesAndVelocity(Clamp Velocity=false, Limit Acceleration=false)
   &FN_UpdateAnimTrail() {module=FN_UpdateAnimTrail001}
   /SetVariables(
    Particles.FrontGlowColor:LinearColor=&Multiply_LinearColorByFloat(
     Linear Color=&LerpLinearColors(
      StartColor=&ColorFromCurve(ColorCurve=DataInterface() {propertyValues {
         AlphaCurve {keys [<at 0 0.15 RCIM_Cubic>,<at 0.3507 0 RCIM_Cubic>]}
         BlueCurve {keys [<at 0 1.6406 RCIM_Cubic>,<at 0.1 0 RCIM_Cubic>]}
         GreenCurve {keys [<at 0 4.5183 RCIM_Cubic>,<at 0.1 1.3429 RCIM_Cubic>]}
         RedCurve {keys [<at 0 15 RCIM_Cubic>,<at 0.1 7.2131 RCIM_Cubic>]}
        }})
      EndColor=&ColorFromCurve(ColorCurve=DataInterface() {propertyValues {
         AlphaCurve {keys [<at 0 0.15 RCIM_Cubic>,<at 0.3507 0 RCIM_Cubic>]}
         BlueCurve {keys [<at 0 0 RCIM_Cubic>,<at 0.1 0 RCIM_Cubic>]}
         GreenCurve {keys [<at 0 3.0829 RCIM_Cubic>,<at 0.1 1.3429 RCIM_Cubic>]}
         RedCurve {keys [<at 0 15 RCIM_Cubic>,<at 0.1 7.2131 RCIM_Cubic>]}
        }})
      LerpFactor=&FN_FrontEndFloatModifier(Gameplay=0, Front End=1)
     )
     Float=&FNTimeOfDayScalar(Night Scale=1, Morning Scale=0.5, Evening Scale=0.5)
    )
    Particles.FrontGlowWidth:NiagaraFloat=&Multiply_Float(A=25, B=$User.WidthMult)
   ) {isSetParameters=true}
  ]
  rendererClasses [&NiagaraRibbonRendererProperties]
  renderers [{rendererClass=&NiagaraRibbonRendererProperties}]
 }
}
summary {
 systemName=NS_Pickaxe_SpireTend_Trail
 emitters [
  {emitterName=Trail, rendererClasses [&NiagaraRibbonRendererProperties]}
  {emitterName=Minimal, rendererClasses [&NiagaraMeshRendererProperties]}
 ]
 userVariables [
  User.bIsFrontEnd {type=NiagaraBool}
  User.bIsFrontEndPreview {type=NiagaraBool}
  User.bIsSwinging {type=NiagaraBool}
  User.RibbonIDOffset {type=NiagaraInt32}
  User.SkeletalMesh {
   type=NiagaraDataInterfaceSkeletalMesh
   defaultValue {
    struct=NiagaraExt_VariableValue_DataInterface
    value {dataInterface=&self:NiagaraDataInterfaceSkeletalMesh_0}
   }
  }
  User.SwingSectionIndex {type=NiagaraInt32}
  User.TrailLifeTimeMult {type=NiagaraFloat, defaultValue=0.25, description=""}
  User.WidthMult {type=NiagaraFloat, defaultValue=2.6, description=""}
 ]
}
