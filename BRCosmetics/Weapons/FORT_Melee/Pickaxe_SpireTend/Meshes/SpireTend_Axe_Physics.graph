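# Physics asset for the SpireTend pickaxe, a two-handed melee cosmetic with a dangling chain.
# Six dyn_hain bodies, five capsules and a box at the end, link in sequence from base_hain through limited swing and tight twist constraints.
# The handle has its own capsule.
# Thigh and calf capsules for both legs, joined by free-rotation constraints, are also present, likely as colliders for the chain.
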PhysicsAsset /BRCosmetics/Weapons/FORT_Melee/Pickaxe_SpireTend/Meshes/SpireTend_Axe_Physics
bodies {
 Melee_TwoHanded_Handle [{shapeType=Capsule, length=9.5, radius=16, shapeName=Melee_TwoHanded_Handle_capsule, center=<0 0 -3.3072>}]
 base_hain [{shapeType=Capsule, length=0, radius=2, shapeName=base_hain_capsule}]
 calf_l [{shapeType=Capsule, length=50, radius=8, center=<-30 -2 0>, rotation=<rot 90 0 0>}]
 calf_r [{shapeType=Capsule, length=50, radius=8, center=<30 2 0>, rotation=<rot 90 0 0>}]
 dyn_hain_01 [{shapeType=Capsule, length=4, radius=3, shapeName=dyn_hain_01_capsule, center=<3 0 0>, rotation=<rot 90 0 0>}]
 dyn_hain_02 [{shapeType=Capsule, length=4, radius=3, shapeName=dyn_hain_02_capsule, center=<3 0 0>, rotation=<rot 90 0 0>}]
 dyn_hain_03 [{shapeType=Capsule, length=4, radius=3, shapeName=dyn_hain_03_capsule, center=<3 0 0>, rotation=<rot 90 0 0>}]
 dyn_hain_04 [{shapeType=Capsule, length=4, radius=3, shapeName=dyn_hain_04_capsule, center=<3 0 0>, rotation=<rot 90 0 0>}]
 dyn_hain_05 [{shapeType=Capsule, length=4, radius=3, shapeName=dyn_hain_05_capsule, center=<3 0 0>, rotation=<rot 90 0 0>}]
 dyn_hain_06 [{shapeType=Box, extentX=2.01, extentY=11.0558, extentZ=11.7485, shapeName=dyn_hain_06_box, center=<4.3175 0 1.6417>, rotation=<rot 0 90 0>}]
 thigh_l [{shapeType=Capsule, length=40, radius=8, center=<-25 0 0>, rotation=<rot 90 0 0>}]
 thigh_r [{shapeType=Capsule, length=40, radius=8, center=<25 0 0>, rotation=<rot 90 0 0>}]
}
constraints [
 {bone1Name=dyn_hain_01, bone2Name=base_hain, swing1LimitDegrees=5, swing2LimitDegrees=35, twistLimitDegrees=5}
 {bone1Name=dyn_hain_02, bone2Name=dyn_hain_01, swing1LimitDegrees=5, swing2LimitDegrees=35, twistLimitDegrees=5}
 {bone1Name=dyn_hain_03, bone2Name=dyn_hain_02, swing1LimitDegrees=35, swing2LimitDegrees=35, twistLimitDegrees=5}
 {bone1Name=dyn_hain_04, bone2Name=dyn_hain_03, swing1LimitDegrees=35, swing2LimitDegrees=35, twistLimitDegrees=5}
 {bone1Name=dyn_hain_05, bone2Name=dyn_hain_04, swing1LimitDegrees=45, swing2LimitDegrees=20, twistLimitDegrees=5}
 {bone1Name=dyn_hain_06, bone2Name=dyn_hain_05, swing1LimitDegrees=45, swing2LimitDegrees=45, twistLimitDegrees=5}
 {
  bone1Name=calf_l
  bone2Name=thigh_l
  swing1LimitDegrees=45
  swing1Motion=Free
  swing2LimitDegrees=45
  swing2Motion=Free
  twistLimitDegrees=45
  twistMotion=Free
 }
 {
  bone1Name=calf_r
  bone2Name=thigh_r
  swing1LimitDegrees=45
  swing1Motion=Free
  swing2LimitDegrees=45
  swing2Motion=Free
  twistLimitDegrees=45
  twistMotion=Free
 }
]
